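# Niagara system for the MeteorWomen alternate glider's flight effects: wing trails, sparks and speed lines.
# Bone-pair anim-trail ribbons and thin eroded ribbons follow the skeletal mesh, with trailing velocity scaled by User.ForwardDot, while sprites add sparks at bones and speed-stretched streaks in a cylinder.
# A system Timeline value gates gameplay opacity and streak spawn rate.
# Front-end modifiers retune motion and fades for menus, and time-of-day scalars adjust trail and spark intensity.
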NiagaraSystem /BRCosmetics/Gadgets/Assets/VinderTech_GliderChute/Glider_Meteorwomen_Alt/FX/NS_Glider_MeteorWomen
dependencies {
 usedDataInterfaces [
  &NiagaraDataInterfaceCurve
  &NiagaraDataInterfaceArrayFloat3
  &NiagaraDataInterfaceArrayColor
  &NiagaraDataInterfaceVector2DCurve
  &NiagaraDataInterfaceColorCurve
 ]
 usedDynamicInputs [&SetBoolByFloatComparison, &Abs_Float]
 usedModules [&SystemState, &Timeline]
}
emitters {
 Rainbo_Ribbon_Faded {
  emitterUpdateScript [
   &EmitterState(Life Cycle Mode=::NewEnumerator1, Loop Duration=0.1, UseLoopDelay=false)
   &FN_SpawnAnimTrail_v2(AnimTrailWidthMode=::FromFirst, Number of Anim Trails=2)
  ]
  particleSpawnScript [
   /Game/Effects/Niagara/Modules/AnimTrail/FN_InitializeAnimTrail/FN_InitializeAnimTrailBoneIndecies(Second Bone or Socket Index=1)
   &FN_InitializeAnimTrailBoneIndecies(AnimTrail Selection=1, First Bone or Socket Index=2, Second Bone or Socket Index=3) {module=FN_InitializeAnimTrailBoneIndecies001}
   /Game/Effects/Niagara/Modules/AnimTrail/FN_InitializeAnimTrail/FN_InitializeAnimTrail(AnimTrailWidth=1, Color=<rgba 0.923 0.9558 1 0.1>, Spawn Mode:ENiagaraSkeletalSpawning=::NewEnumerator0, Write Material Random=false)
   &AddVelocity(Velocity=&MakeVector(
     X=0
     Y=&FloatFromCurve(
      FloatCurve=DataInterface() {propertyValues {Curve {keys [<at 0 0.25>,<at 0.5 1>]}}}
      CurveIndex=&RemapRange(Input Value To Remap=$User.ForwardDot, Input Min=-1, Input Max=1, Output Min=0, Output Max=1)
      Scale Curve=&FN_FrontEndFloatModifier(Gameplay=-500, Front End=-400)
     )
     Z=0
    ))
  ]
  particleUpdateScript [
   &InheritSourceMovement(
    Applied Position Scale=&Multiply_VectorByFloat(
     Vector=<0.7 0.7 1>
     Float=&FloatFromCurve(FloatCurve=DataInterface() {propertyValues {Curve {keys [{interpMode=RCIM_Cubic, tangentMode=RCTM_User, time=0, value=1}, {interpMode=RCIM_Cubic, tangentMode=RCTM_User, time=1, value=0.3}]}}})
    )
    ApplyForce=false
    ApplyPosition=true
    ApplyPositionFalloff=false
   )
   &AccelerationForce(
    Acceleration=&MakeVector(
     X=0
     Y=0
     Z=&MakeCustomFloatFromBool(
      Bool=$User.bIsFrontEnd
      True Float=&FloatFromCurve(FloatCurve=DataInterface() {propertyValues {Curve {keys [<at 0 -1>,<at 1 2.5>]}}}, Scale Curve=2000)
      False Float=0
     )
    )
    Coordinate Space=::Simulation
   )
   /SetVariables(Particles.Velocity=&Add_Vector(
     A=$Particles.Velocity
     B=&MakeCustomVectorFromBool(
      Bool=$User.bIsFrontEnd
      True Vector=&RotateVectorByQuaternion(
       Vector To Rotate=&MakeVector(
        X=&FloatFromCurve(FloatCurve=DataInterface() {propertyValues {Curve {keys [<at 0 -0.3>,<at 1 0.8>]}}}, Scale Curve=200)
        Y=0
        Z=0
       )
       Quaternion=$Engine.Owner.Rotation
      )
      False Vector=<1 0 0>
     )
    )) {enabled=false, isSetParameters=true}
   &SolveForcesAndVelocity(Clamp Velocity=false, Limit Acceleration=false)
   &FN_UpdateAnimTrail_v2()
   &ScaleRibbonWidth(Ribbon Width Scale=6)
   &FN_RibbonMaterial() {enabled=false}
   &ScaleColor(
    Scale Alpha=&Lerp_Float(
     A=&FloatFromCurve(
      FloatCurve=DataInterface() {propertyValues {Curve {keys [{interpMode=RCIM_Cubic, tangentMode=RCTM_User, time=0, value=1}, {interpMode=RCIM_Cubic, tangentMode=RCTM_User, time=1, value=0}]}}}
      Scale Curve=&Multiply_Float(
       A=$System.Timeline.Time
       B=&FloatFromCurve(
        FloatCurve=DataInterface() {propertyValues {Curve {keys [
            {interpMode=RCIM_Cubic, tangentMode=RCTM_User, time=0, value=1}
            {interpMode=RCIM_Cubic, tangentMode=RCTM_User, time=0.5, value=0}
            {tangentMode=RCTM_User, time=1, value=1}
           ]}}}
        CurveIndex=$System.FN_Integrator.Value
       )
      )
     )
     B=&FloatFromCurve(FloatCurve=DataInterface() {propertyValues {Curve {keys [{interpMode=RCIM_Cubic, tangentMode=RCTM_User, time=0, value=1}, {interpMode=RCIM_Cubic, tangentMode=RCTM_User, time=0.8, value=0}]}}})
     Alpha=&FN_FrontEndFloatModifier(Gameplay=0, Front End=1)
    )
    Scale RGB=&VectorFromFloat(Value=&FNTimeOfDayScalar(Night Scale=1, Morning Scale=1, Day Scale=2, Evening Scale=1))
    ScaleA=true
    ScaleRGB=true
   )
   /SetVariables(Particles.Renderer2_RibbonWidth:NiagaraFloat=&Multiply_Float(A=$Particles.RibbonWidth, B=0.9)) {isSetParameters=true}
   /SetVariables(Particles.Renderer2Position:Vector3f=&Add_Vector(A=Linked() {linkedVariable {name=Particles.Position, type=NiagaraPosition}}, B=<0 0 -5>)) {isSetParameters=true}
  ]
  rendererClasses [&NiagaraRibbonRendererProperties]
  renderers [
   {rendererClass=&NiagaraRibbonRendererProperties}
   {rendererClass=&NiagaraRibbonRendererProperties, rendererIndex=1}
   {rendererClass=&NiagaraRibbonRendererProperties, rendererIndex=2}
  ]
 }
 Spark_Shape001 {
  emitterUpdateScript [
   &EmitterState()
   &SpawnRate(SpawnRate=40, Use Spawn Probability=false)
   &SpawnBurst_Instantaneous(Spawn Count=40, Spawn Time=0.7, Use Spawn Probability=false) {enabled=false}
  ]
  particleSpawnScript [
   &FN_SetSpawnGroup()
   &InitializeParticle(
    Color=<rgba 4 4 4 0.5>
    Lifetime Max=0.8
    Lifetime Min=0.4
    Lifetime Mode=::NewEnumerator1
    Sprite Rotation Mode=::NewEnumerator2
    Sprite Size Mode=::NewEnumerator1
    Uniform Sprite Size Min=3
    UsePositionOffset=false
    Write Material Random=true
   )
   &SkeletalMeshLocation(
    Bone / Socket Index=$Particles.SpawnGroup
    Bone Sampling Mode=::NewEnumerator4
    MaskSpawnGroup=false
    Sampled Position Offset=&SelectVectorFromArray(
     Direct Array Index=$Particles.SpawnGroup
     Vector Selection Array=DataInterface() {propertyValues {FloatData [<0 0 0>, <0 0 0>, <0 0 0>, <0 0 0>]}}
    )
    Skeletal Mesh=$User.SkeletalMesh
   )
   &ShapeLocation(MaskSpawnGroup=false, Offset=<0 0 -40>, Offset Mode=::NewEnumerator0, Override Seed=false, Sphere Radius=15, UseOwnerScale=true)
   &AddVelocityFromPoint(
    Module.Use Falloff Distance=false
    Velocity Strength=&Multiply_Float(A=0.2, B=&RandomRangeFloat(Minimum=1300, Maximum=2000))
   )
   &AddVelocity(Velocity=<0 -800 0>)
   &CameraOffset(Camera Offset Amount=&RandomRangeFloat(Minimum=5, Maximum=8))
  ]
  particleUpdateScript [
   &ParticleState()
   &ScaleColor(
    Linear Color Curve=DataInterface() {propertyValues {
      AlphaCurve {keys [<at 0.0074 1>,<at 0.8905 1>]}
      BlueCurve {keys [<at 0 8.4167>,<at 0.7 1>]}
      GreenCurve {keys [<at 0 10>,<at 0.7 0.5333>]}
      RedCurve {keys [<at 0 0.5>,<at 0.7 0>]}
     }}
    Scale Mode=::NewEnumerator2
   ) {enabled=false}
   &ScaleColor(
    Scale Alpha=&Multiply_Float(
     A=&Multiply_Float(A=$System.Timeline.Time, B=1)
     B=&FNTimeOfDayScalar(Night Scale=0.5, Morning Scale=0.3, Evening Scale=0.3)
    )
    Scale RGB=&VectorFromFloat(Value=&FloatFromCurve(FloatCurve=DataInterface() {propertyValues {Curve {keys [<at 0 20>,<at 0.2 20>,<at 1 5>]}}}, Scale Curve=0.1))
    ScaleA=true
    ScaleRGB=true
   ) {module=ScaleColor001}
   &Drag(Drag=&RandomRangeFloat(Minimum=4, Maximum=8), Use Linear Drag=true, Use Rotational Drag=false)
   &GravityForce(Gravity=&RandomRangeVector(Minimum=<0 0 0>, Maximum=<0 0 -100>)) {enabled=false}
   &InheritSourceMovement(
    Applied Force Scale=&VectorFromFloat(Value=&FloatFromCurve(FloatCurve=DataInterface() {propertyValues {Curve {keys [<at 0.2 0>,<at 0.3 1>]}}}, Scale Curve=-1))
    Applied Position Falloff Distance=150
    ApplyForce=true
    ApplyForceFalloff=false
    ApplyPosition=true
    ApplyPositionFalloff=false
    Module.LimitForceSpeed=false
   ) {module=InheritSourceMovement003}
   &InheritSourceMovement(
    Applied Force Scale=&VectorFromFloat(Value=&FloatFromCurve(FloatCurve=DataInterface() {propertyValues {Curve {keys [<at 0.4 1>,<at 0.8 0>]}}}))
    ApplyForce=true
    ApplyForceFalloff=false
    ApplyPosition=false
    Module.LimitForceSpeed=false
   ) {enabled=false}
   &SolveForcesAndVelocity(Clamp Velocity=false, Limit Acceleration=false)
   &ScaleSpriteSize(Scale Factor=&Vector2DFromFloat(Value=&FloatFromCurve(FloatCurve=DataInterface() {propertyValues {Curve {keys [
          {interpMode=RCIM_Cubic, tangentMode=RCTM_User, time=0, value=0}
          {arriveTangent=0.4864, interpMode=RCIM_Cubic, leaveTangent=0.4864, tangentMode=RCTM_User, time=0.15, value=0.8}
          {interpMode=RCIM_Cubic, tangentMode=RCTM_User, time=1, value=0}
         ]}}})))
   &DynamicMaterialParameters(Module.Param0WriteEnabled=true, Module.Param1WriteEnabled=true, Module.Param2WriteEnabled=true, Module.Param3WriteEnabled=true) {enabled=false}
  ]
  rendererClasses [&NiagaraSpriteRendererProperties]
  renderers [{rendererClass=&NiagaraSpriteRendererProperties}]
 }
 Speedliness {
  emitterUpdateScript [
   &EmitterState(Enable Distance Culling=true, MaxDistance=2048, Module.bUseMaxDistance=true, Module.bUseMinDistance=false, Scalability Mode=::NewEnumerator1)
   &SpawnRate(
    SpawnRate=&Multiply_Float(
     A=&FN_FrontEndFloatModifier(Gameplay=&Multiply_Float(A=$System.Timeline.Time, B=1.5), Front End=1)
     B=&UniformRangedFloat(Minimum=75, Maximum=150)
    )
    Use Spawn Probability=false
   )
  ]
  particleSpawnScript [
   /Niagara/Modules/Spawn/Initialization/InitializeParticle(
    Color=&MakeLinearColorFromVectorAndFloat(
     "Vector (RGB)"=<1 1 1>
     "Float (Alpha)"=&Multiply_Float(
      A=&Multiply_Float(
       A=&UniformRangedFloat(Minimum=0.015, Maximum=0.15)
       B=&FN_FrontEndFloatModifier(Gameplay=1, Front End=0.5)
      )
      B=&Add_Float(
       A=1
       B=&Multiply_Float(A=10, B=&Subtract_Float(A=1, B=$Engine.Emitter.SpawnCountScale))
      )
     )
    )
    Lifetime=&FN_FrontEndFloatModifier(
     Gameplay=&UniformRangedFloat(Minimum=0.2, Maximum=0.35)
     Front End=&UniformRangedFloat(Minimum=0.1, Maximum=0.2)
    )
    Module.Write Color=true
    Module.Write Lifetime=true
    Module.Write Mass=true
    Module.Write Material Random=true
    Module.Write Position=false
    Module.Write Scale=false
    Module.Write SpriteRotation=false
    Module.Write SpriteSize=true
    Position=&SimulationPosition()
    Sprite Size=&UniformRangedVector2D(Minimum=<3 4>, Maximum=<5 6>)
   )
   &CylinderLocation(
    Cylinder Radius=&FN_FrontEndFloatModifier(
     Gameplay=&UniformRangedFloat(Minimum=100, Maximum=150)
     Front End=&UniformRangedFloat(Minimum=150, Maximum=200)
    )
    MaskSpawnGroup=false
    Offset Coordinate Space:ENiagaraCoordinateSpace=::Local
    Offset Mode:ENiagara_OffsetMode=::NewEnumerator0
    Override Seed=false
    UseOwnerScale=true
   )
   &AddVelocity(Velocity=&FN_FrontEndVector3Modifier(Gameplay=<0 -1000 0>, Front End=<0 -1111 0>))
   &AddVelocityFromPoint(
    Module.Use Falloff Distance=false
    Velocity Strength=&Multiply_Float(
     A=&FN_FrontEndFloatModifier(Gameplay=1, Front End=0.5)
     B=&UniformRangedFloat(Minimum=500, Maximum=700)
    )
   )
  ]
  particleUpdateScript [
   &ParticleState()
   &SolveForcesAndVelocity(Clamp Velocity=false, Limit Acceleration=false)
   &ScaleSpriteSize(Scale Factor=&Vector2DFromCurve(
     Vector2Curve=DataInterface() {propertyValues {XCurve {keys [<at 0 0.5 RCIM_Cubic>,<at 0.5 1 RCIM_Cubic>]}, YCurve {keys [<at 0 0 RCIM_Cubic>,<at 0.5 1 RCIM_Cubic>]}}}
     ScaleCurve=<1 0.35>
    ))
   &ScaleSpriteSizeBySpeed(Max Scale Factor=<1 40>, Min Scale Factor=<0.01 0.05>, Velocity Threshold=1)
   &ScaleColor(
    Scale Alpha=&FloatFromCurve(
     FloatCurve=DataInterface() {propertyValues {Curve {keys [<at 0 0 RCIM_Cubic>, <at 0.2 1>, {arriveTangent=-2.221, interpMode=RCIM_Cubic, leaveTangent=-2.221, tangentMode=RCTM_User, time=1, value=0}]}}}
     Scale Curve=&FN_FrontEndFloatModifier(Gameplay=0.4, Front End=1)
    )
    ScaleA=true
    ScaleRGB=true
   )
   &CameraOffset(Camera Offset Amount=&FN_FrontEndFloatModifier(Gameplay=0, Front End=-20))
   &ScaleColor(
    Scale Alpha=&MakeCustomFloatFromBool(Bool=$User.bIsFrontEnd, True Float=1, False Float=&Multiply_Float(A=1, B=$System.Timeline.Time))
    ScaleA=true
    ScaleRGB=true
   ) {module=ScaleColor002}
  ]
  rendererClasses [&NiagaraSpriteRendererProperties]
  renderers [{rendererClass=&NiagaraSpriteRendererProperties}]
 }
 ThinRibbonsOptimized {
  emitterUpdateScript [
   &EmitterState()
   &FN_SpawnRibbon_v2(Number of Ribbons=6, Spawn Mode=::NewEnumerator0, SpawnRate=30)
   &FN_Integrator() {enabled=false}
  ]
  particleSpawnScript [
   /Game/Effects/Niagara/Modules/Ribbons/FN_InitializeRibbon_v4(
    Color:LinearColor=<rgba 1 1 1 0.7>
    Enable Skeletal Position=true
    Offset Amount=5
    Ribbon Width=25
    SkeletalMesh:NiagaraDataInterfaceSkeletalMesh=$User.SkeletalMesh
   )
   &FN_FlipRibbonV()
   &AddVelocity(Velocity=<0 0 100>) {enabled=false}
   &AddVelocity(Velocity=&MakeVector(
     X=0
     Y=&FloatFromCurve(
      FloatCurve=DataInterface() {propertyValues {Curve {keys [<at 0 0.25>,<at 0.5 1>]}}}
      CurveIndex=&RemapRange(Input Value To Remap=$User.ForwardDot, Input Min=-1, Input Max=1, Output Min=0, Output Max=1)
      Scale Curve=-500
     )
     Z=0
    )) {module=AddVelocity001}
   /SetVariables(Particles.Position=&Add_Vector(
     A=Linked() {linkedVariable {name=Particles.Position, type=NiagaraPosition}}
     B=&SelectVectorFromArray(
      Direct Array Index=&ReturnParticleIDAttributes(Particle ID Attribute=::NewEnumerator1, Niagara ID=$Particles.RibbonID)
      Vector Selection Array=DataInterface() {propertyValues {FloatData [<0 -100 0>, <0 -100 0>, <0 -100 0>, <0 -100 0>, <0 -100 0>, <0 0 0>, <0 0 0>, <0 0 0>, <0 0 0>, <0 0 0>]}}
     )
    )) {enabled=false, isSetParameters=true}
   &AddVelocity(Velocity=<0 4000 50>) {enabled=false, module=AddVelocity002}
  ]
  particleUpdateScript [
   &InheritSourceMovement(
    Applied Position Scale=&Multiply_VectorByFloat(
     Vector=<0.7 0.7 1>
     Float=&FloatFromCurve(FloatCurve=DataInterface() {propertyValues {Curve {keys [{interpMode=RCIM_Cubic, tangentMode=RCTM_User, time=0, value=1}, {interpMode=RCIM_Cubic, tangentMode=RCTM_User, time=1, value=0.3}]}}})
    )
    ApplyForce=false
    ApplyPosition=true
    ApplyPositionFalloff=false
   )
   &AccelerationForce(
    Acceleration=&MakeVector(
     X=0
     Y=0
     Z=&MakeCustomFloatFromBool(
      Bool=$User.bIsFrontEnd
      True Float=&FloatFromCurve(FloatCurve=DataInterface() {propertyValues {Curve {keys [<at 0 -1>,<at 1 3>]}}}, Scale Curve=2200)
      False Float=0
     )
    )
    Coordinate Space=::Simulation
   )
   /SetVariables(Particles.Velocity=&Add_Vector(
     A=$Particles.Velocity
     B=&MakeCustomVectorFromBool(
      Bool=$User.bIsFrontEnd
      True Vector=&RotateVectorByQuaternion(
       Vector To Rotate=&MakeVector(
        X=&FloatFromCurve(FloatCurve=DataInterface() {propertyValues {Curve {keys [<at 0 -0.3>,<at 1 0.8>]}}}, Scale Curve=200)
        Y=0
        Z=0
       )
       Quaternion=$Engine.Owner.Rotation
      )
      False Vector=<1 0 0>
     )
    )) {enabled=false, isSetParameters=true}
   &SolveForcesAndVelocity(Clamp Velocity=false, Limit Acceleration=false)
   /Game/Effects/Niagara/Modules/Ribbons/FN_UpdateRibbon_v4(
    Enable Skeletal Position=true
    SkeletalMesh=DataInterface() {propertyValues {
      DefaultMesh=null
      ExcludeBoneName=root
      OptionalComponentClass=null
      PreviewMesh=/BRCosmetics/Gadgets/Assets/VinderTech_GliderChute/Glider_Binary_Female/Meshes/Binary_Female
      SkinningMode=SkinOnTheFly
      SoftSourceActor=null
      SourceMode=Default
      UvSetIndex=0
      WholeMeshLOD=-1
      bExcludeBone=false
      bReadDeformedGeometry=true
      bRequireCurrentFrameData=true
      ComponentTags []
      FilteredBones [lt_wingMini_01_JNT, lt_wingMini_02_JNT, lt_wingMini_03_JNT, rt_wingMini_01_JNT, rt_wingMini_02_JNT, rt_wingMini_03_JNT]
      FilteredSockets []
      MeshUserParameter {parameter {name=None, typeDefHandle {registeredTypeIndex=1}, varData []}}
      SamplingRegions []
     }}
   )
   &FN_RibbonMaterial(
    Opacity Erosion:NiagaraFloat=&FloatFromCurve(FloatCurve=DataInterface() {propertyValues {Curve {keys [
         {interpMode=RCIM_Cubic, tangentMode=RCTM_User, time=0, value=0}
         {arriveTangent=2.6226, interpMode=RCIM_Cubic, leaveTangent=2.6226, tangentMode=RCTM_User, time=1, value=1}
        ]}}})
    Opacity Pan U Offset:NiagaraFloat=$Particles.MaterialRandom
    Opacity Pan U Speed=1.5
    Opacity Scale U=1.25
   )
   &ScaleRibbonWidth(Ribbon Width Scale=&Multiply_Float(
     A=&FloatFromCurve(
      FloatCurve=DataInterface() {propertyValues {Curve {keys [
          {arriveTangent=12.569, interpMode=RCIM_Cubic, leaveTangent=12.569, tangentMode=RCTM_User, time=0, value=0}
          <at 0.1 0.95>
          {interpMode=RCIM_Cubic, tangentMode=RCTM_User, time=1, value=0}
         ]}}}
      Scale Curve=&FN_FrontEndFloatModifier(Gameplay=2, Front End=1)
     )
     B=&RemapRange(Input Value To Remap=$Particles.MaterialRandom, Input Min=0, Input Max=1, Output Min=0.9, Output Max=1.2)
    ))
   /SetVariables(Particles.Color=&SelectLinearColorFromArray(
     Array Sampling Mode=::NewEnumerator1
     Direct Array Index=&ReturnParticleIDAttributes(Particle ID Attribute=::NewEnumerator1, Niagara ID=$Particles.RibbonID)
     Color Selection Array=DataInterface() {propertyValues {ColorData [
        <rgba 1 0 0.003 1>
        <rgba 1 0.0893 0 1>
        <rgba 1 0.2137 0 1>
        <rgba 0.1658 1 0 1>
        <rgba 0 1 0.7493 1>
        <rgba 0 1 0.6372 1>
        <rgba 0 0.373 1 1>
        <rgba 0 1 0.6393 1>
        <rgba 0 0.2813 1 1>
        <rgba 0.1218 0 1 1>
        <rgba 1 0 0.0334 1>
        <rgba 1 0.1961 0 1>
        <rgba 1 0.572 0 1>
        <rgba 0.1979 1 0 1>
        <rgba 0 1 0.5545 1>
        <rgba 0 1 0.2446 1>
        <rgba 0 0.7521 1 1>
        <rgba 0 1 0.618 1>
        <rgba 0 1 0.8868 1>
        <rgba 0.2717 0 1 1>
       ]}}
     Random Seed=?
    )) {enabled=false, isSetParameters=true}
   &ScaleColor(
    Scale Alpha=&Lerp_Float(
     A=&Multiply_Float(
      A=$System.Timeline.Time
      B=&FloatFromCurve(
       FloatCurve=DataInterface() {propertyValues {Curve {keys [
           {arriveTangent=1.0365, interpMode=RCIM_Cubic, leaveTangent=1.0365, tangentMode=RCTM_User, time=0, value=0}
           {interpMode=RCIM_Cubic, tangentMode=RCTM_User, time=0.3, value=0.95}
           {tangentMode=RCTM_User, time=1, value=0}
          ]}}}
       Scale Curve=0.5
      )
     )
     B=&Multiply_Float(
      A=1
      B=&FloatFromCurve(
       FloatCurve=DataInterface() {propertyValues {Curve {keys [
           {arriveTangent=1.0365, interpMode=RCIM_Cubic, leaveTangent=1.0365, tangentMode=RCTM_User, time=0, value=0}
           {interpMode=RCIM_Cubic, tangentMode=RCTM_User, time=0.3, value=0.95}
           {tangentMode=RCTM_User, time=1, value=0}
          ]}}}
       Scale Curve=0.7
      )
     )
     Alpha=&FN_FrontEndFloatModifier(Gameplay=0, Front End=1)
    )
    Scale RGB=&VectorFromFloat(Value=0.4)
    ScaleA=true
    ScaleRGB=true
   )
   &ScaleColor(
    Scale Alpha=&FloatFromCurve(
     FloatCurve=DataInterface() {propertyValues {Curve {keys [{interpMode=RCIM_Cubic, tangentMode=RCTM_User, time=0, value=1}, {interpMode=RCIM_Cubic, tangentMode=RCTM_User, time=1, value=0}]}}}
     Scale Curve=&FloatFromCurve(
      FloatCurve=DataInterface() {propertyValues {Curve {keys [
          {interpMode=RCIM_Cubic, tangentMode=RCTM_User, time=0, value=1}
          {interpMode=RCIM_Cubic, tangentMode=RCTM_User, time=0.5, value=0}
          {tangentMode=RCTM_User, time=1, value=1}
         ]}}}
      CurveIndex=$System.FN_Integrator.Value
     )
    )
    Scale RGB=&VectorFromFloat(Value=&FNTimeOfDayScalar(Night Scale=1, Morning Scale=1, Day Scale=2, Evening Scale=1))
    ScaleA=true
    ScaleRGB=true
   ) {enabled=false, module=ScaleColor001}
  ]
  rendererClasses [&NiagaraRibbonRendererProperties]
  renderers [{rendererClass=&NiagaraRibbonRendererProperties}]
 }
}
summary {
 systemName=NS_Glider_MeteorWomen
 emitters [
  {emitterName=Rainbo_Ribbon_Faded, rendererClasses [&NiagaraRibbonRendererProperties]}
  {emitterName=ThinRibbonsOptimized, rendererClasses [&NiagaraRibbonRendererProperties]}
  {emitterName=Speedliness, rendererClasses [&NiagaraSpriteRendererProperties]}
  {emitterName=Spark_Shape001, rendererClasses [&NiagaraSpriteRendererProperties]}
 ]
 userVariables [
  User.bIsFrontEnd {type=NiagaraBool}
  User.bIsFrontEndPreview {type=NiagaraBool}
  User.bisFullyDeployed {type=NiagaraBool}
  User.ForwardDot {type=NiagaraFloat, defaultValue=0, description=""}
  User.RibbonIDOffset {type=NiagaraInt32}
  User.SkeletalMesh {
   type=NiagaraDataInterfaceSkeletalMesh
   defaultValue {
    struct=NiagaraExt_VariableValue_DataInterface
    value {dataInterface=&self:NiagaraDataInterfaceSkeletalMesh_1}
   }
  }
  User.Variant {type=NiagaraFloat, defaultValue=0, description=""}
 ]
}
